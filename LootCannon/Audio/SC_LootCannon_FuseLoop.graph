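# SoundCue for the Loot Cannon's looping fuse sound, routed to the explosives sound class at reduced volume.
# A looping TNT fuse wave feeds an Enveloper that quickly fades volume in from silence, then a Modulator that randomizes pitch upward while leaving volume unchanged.
# It references a shared explosive-fuse ambient attenuation asset and overrides concurrency to cap simultaneous instances.
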SoundCue /LootCannon/Audio/SC_LootCannon_FuseLoop
root=Modulator_1
nodes [
 {
  id=Enveloper_0
  children [WavePlayer_0]
  properties {
   pitchCurve {editorCurveData {keys [{time=0}]}}
   volumeCurve {editorCurveData {keys [{arriveTangent=0.9158, interpMode=RCIM_Cubic, leaveTangent=0.9158, tangentMode=RCTM_User, time=0, value=0}, <at 0.302 1.0018>, <at 1 1>]}}
  }
 }
 {id=Modulator_1, children [Enveloper_0], properties {pitchMax=1.3, pitchMin=1, volumeMax=1, volumeMin=1}}
 {id=WavePlayer_0, properties {bLooping=true, soundWaveAssetPtr=/LootCannon/Audio/SoundWaves/TNT_Fuse_Loop01}}
]
properties {
 attenuationSettings=/Game/Sounds/Attenuation/Fort_Explosive_Fuse_Ambient_Att
 bOverrideConcurrency=true
 maxDistance=3500
 soundClassObject=/Game/Sounds/FortSoundClassesAndModes/Fort_Explosives
 volumeMultiplier=0.4
 attenuationOverrides {bEnableListenerFocus=true, falloffDistance=2500, nonFocusVolumeAttenuation=0.6, stereoSpread=0, attenuationShapeExtents {x=1500}}
 concurrencyOverrides {maxCount=3}
}
