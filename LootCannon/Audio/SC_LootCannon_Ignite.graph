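# Sound cue for the loot cannon's ignition, layering a cannon ignite sound with TNT ignite variations.
# The cannon wave plays immediately through slight pitch randomization and an enveloper with a rising pitch curve.
# A randomly chosen, pitch-varied TNT wave follows after a short fixed delay, mixed louder, and an outer mixer boosts both.
# It uses small melee weapon attenuation and class, and allows one instance per owner, blocking new plays.
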SoundCue /LootCannon/Audio/SC_LootCannon_Ignite
root=Mixer_1
nodes [
 {id=Delay_0, children [Modulator_0], properties {delayMax=0.4, delayMin=0.4}}
 {
  id=Enveloper_0
  children [Modulator_2]
  properties {pitchCurve {editorCurveData {keys [{time=0}, <at 1 2>]}}, volumeCurve {editorCurveData {keys [<at 0 1>]}}}
 }
 {id=Mixer_0, children [Delay_0,Enveloper_0], properties {inputVolume [2,1]}}
 {id=Mixer_1, children [Mixer_0], properties {inputVolume [2]}}
 {id=Modulator_0, children [Random_0], properties {pitchMax=1.1, pitchMin=0.9}}
 {id=Modulator_2, children [WavePlayer_0], properties {pitchMax=1.2, pitchMin=1}}
 {id=Random_0, children [WavePlayer_1,WavePlayer_2,WavePlayer_3,WavePlayer_4], properties {weights [1,1,1,1]}}
 {id=WavePlayer_0, properties {soundWaveAssetPtr=/LootCannon/Audio/SoundWaves/LootCannon_Ignite}}
 {id=WavePlayer_1, properties {soundWaveAssetPtr=/LootCannon/Audio/SoundWaves/TNT_Ignite_01}}
 {id=WavePlayer_2, properties {soundWaveAssetPtr=/LootCannon/Audio/SoundWaves/TNT_Ignite_02}}
 {id=WavePlayer_3, properties {soundWaveAssetPtr=/LootCannon/Audio/SoundWaves/TNT_Ignite_03}}
 {id=WavePlayer_4, properties {soundWaveAssetPtr=/LootCannon/Audio/SoundWaves/TNT_Ignite_04}}
]
properties {
 attenuationSettings=&Fort_Weapon_Melee_Sml_Att
 bOverrideConcurrency=true
 maxDistance=2800
 soundClassObject=&Fort_WeaponMelee_PSM
 volumeMultiplier=1
 concurrencyOverrides {bLimitToOwner=true, maxCount=1, resolutionRule=PreventNew, retriggerTime=0.2}
}
