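# Sound cue for the Loot Cannon's loot launch, layering close and medium-distance explosion-style sounds.
# The close layer mixes a random sweetener with a pitch-varied random close launch under close explosion attenuation.
# The medium layer plays a pitch-varied random launch with grenade-explosion attenuation and a medium explosives sound class override.
# A top mixer blends both layers, and concurrency caps instances by stopping the oldest.
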SoundCue /LootCannon/Audio/SC_LootCannon_Loot_Launch
root=Mixer_4
nodes [
 {
  id=Attenuation_5
  children [Modulator_3]
  properties {
   attenuationSettings=/Game/Sounds/Attenuation/Fort_Explosion_Medium_Athena_Grenade_Att
   attenuationOverrides {falloffDistance=3000, attenuationShapeExtents {x=1000}}
  }
 }
 {
  id=Attenuation_6
  children [Mixer_1]
  properties {attenuationSettings=/Game/Sounds/Attenuation/Fort_Explosion_Close_Athena_Att, attenuationOverrides {falloffDistance=1600}}
 }
 {id=Mixer_1, children [Random_2,Modulator_6], properties {inputVolume [1,0.5]}}
 {id=Mixer_4, children [Attenuation_6, SoundClass_0], properties {inputVolume [1.2,0.7]}}
 {id=Modulator_3, children [Random_5], properties {pitchMax=1, pitchMin=0.9}}
 {id=Modulator_6, children [Random_4], properties {pitchMax=1.1, pitchMin=1}}
 {id=Random_2, children [WavePlayer_10, WavePlayer_11], properties {weights [1,1]}}
 {id=Random_4, children [WavePlayer_12, WavePlayer_13, WavePlayer_14], properties {weights [1,1,1]}}
 {id=Random_5, children [WavePlayer_15, WavePlayer_16, WavePlayer_17], properties {weights [1,1,1]}}
 {id=SoundClass_0, children [Attenuation_5], properties {soundClassOverride=/Game/Sounds/FortSoundClassesAndModes/Fort_Explosives_Medium}}
 {id=WavePlayer_10, properties {soundWaveAssetPtr=/LootCannon/Audio/SoundWaves/LootCannon_Launch_Sweetener_01}}
 {id=WavePlayer_11, properties {soundWaveAssetPtr=/LootCannon/Audio/SoundWaves/LootCannon_Launch_Sweetener_02}}
 {id=WavePlayer_12, properties {soundWaveAssetPtr=/LootCannon/Audio/SoundWaves/LootCannon_Launch_Close_01}}
 {id=WavePlayer_13, properties {soundWaveAssetPtr=/LootCannon/Audio/SoundWaves/LootCannon_Launch_Close_02}}
 {id=WavePlayer_14, properties {soundWaveAssetPtr=/LootCannon/Audio/SoundWaves/LootCannon_Launch_Close_03}}
 {id=WavePlayer_15, properties {soundWaveAssetPtr=/LootCannon/Audio/SoundWaves/LootCannon_Launch_Med_01}}
 {id=WavePlayer_16, properties {soundWaveAssetPtr=/LootCannon/Audio/SoundWaves/LootCannon_Launch_Med_02}}
 {id=WavePlayer_17, properties {soundWaveAssetPtr=/LootCannon/Audio/SoundWaves/LootCannon_Launch_Med_03}}
]
properties {
 bOverrideConcurrency=true
 maxDistance=9625
 soundClassObject=/Game/Sounds/FortSoundClassesAndModes/Fort_Explosives
 volumeMultiplier=1
 attenuationOverrides {reverbWetLevelMax=0.7, reverbWetLevelMin=0.2}
 concurrencyOverrides {maxCount=4, resolutionRule=StopOldest, volumeScale=0.5}
}
