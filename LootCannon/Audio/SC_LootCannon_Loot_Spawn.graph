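# Sound cue for loot spawning from the Loot Cannon, layering a treasure chest sting over a medkit landing sound.
# A mixer blends a random, louder landing wave with a quieter random chest wave, which is pitched up and briefly delayed, then an outer mixer boosts the result.
# It uses a loot reward sound class, overridden attenuation with occlusion, and concurrency that prevents new instances at the limit.
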SoundCue /LootCannon/Audio/SC_LootCannon_Loot_Spawn
root=Mixer_0
nodes [
 {id=Delay_0, children [Modulator_0], properties {delayMax=0.2, delayMin=0.2}}
 {id=Mixer_0, children [Mixer_1], properties {inputVolume [2]}}
 {id=Mixer_1, children [Delay_0,null,Random_1], properties {inputVolume [0.3,1,2]}}
 {id=Modulator_0, children [Random_0], properties {pitchMax=1.2, pitchMin=1.1}}
 {id=Random_0, children [WavePlayer_0,WavePlayer_1,WavePlayer_2], properties {weights [1,1,1]}}
 {id=Random_1, children [WavePlayer_8, WavePlayer_9, WavePlayer_10], properties {weights [1,1,1]}}
 {id=WavePlayer_0, properties {soundWaveAssetPtr=/LootCannon/Audio/SoundWaves/treasure_chest_sweet_shorter_01}}
 {id=WavePlayer_1, properties {soundWaveAssetPtr=/LootCannon/Audio/SoundWaves/treasure_chest_sweet_shorter_02}}
 {id=WavePlayer_10, properties {soundWaveAssetPtr=/LootCannon/Audio/SoundWaves/Medkit_Land_01}}
 {id=WavePlayer_2, properties {soundWaveAssetPtr=/LootCannon/Audio/SoundWaves/treasure_chest_sweet_shorter_03}}
 {id=WavePlayer_8, properties {soundWaveAssetPtr=/LootCannon/Audio/SoundWaves/Medkit_Land_03}}
 {id=WavePlayer_9, properties {soundWaveAssetPtr=/LootCannon/Audio/SoundWaves/Medkit_Land_02}}
]
properties {
 bOverrideAttenuation=true
 bOverrideConcurrency=true
 maxDistance=3500
 soundClassObject=&Fort_Loot_Reward_PSM
 volumeMultiplier=1
 attenuationOverrides {
  bEnableOcclusion=true
  falloffDistance=2500
  nonSpatializedRadiusStart=150
  occlusionTraceChannel=ECC_GameTraceChannel11
  attenuationShapeExtents {x=1000}
 }
 concurrencyOverrides {resolutionRule=PreventNew}
}
